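# Translucent unlit, two-sided glow material for the TimeInterval pickaxe's Niagara sprite and mesh particle effects.
# Emissive multiplies a color parameter by particle color, then runs through MF_FX_EmissiveAdjustments.
# Opacity combines particle alpha with a grayscale glow texture, or a chosen packed-texture channel via static switch.
# A smoothstep on absolute camera-to-vertex-normal dot product fades it by view angle, and DepthFade softens intersections.
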Material /BRCosmetics/Weapons/FORT_Melee/Pickaxe_TimeInterval/FX/M_Pickaxe_TimeInterval_Glow
graph {
 properties {
  bUsedWithNiagaraMeshParticles=true
  bUsedWithNiagaraSprites=true
  blendMode=BLEND_Translucent
  pixelDepthOffsetMode=PDOM_Legacy
  shadingModel=MSM_Unlit
  twoSided=true
  usedShadingModels=Unlit
 }
 nodes [
  n1 = ParticleColor()
  n2 = TextureSampleParameter2D() {
   parameterName=Glow Texture
   samplerType=SAMPLERTYPE_LinearGrayscale
   texture=&/T_Pickaxe_TimeInterval_Glow
  }
  n3 = StaticSwitchParameter(
   True=ChannelMaskParameter(None=TextureSampleParameter2D() {parameterName=PackedGlowTexture, texture=/Game/UI/GreyBox/white}.RGBA) {parameterName=ChannelToUse}.R
   False=@n2.R
  ) {parameterName=UsePackedChannels}
  n4 = ScalarParameter() {parameterName=View Angle Fade Min, sliderMax=1}
  n5 = ScalarParameter() {defaultValue=1, parameterName=View Angle Fade Max, sliderMax=1}
  n6 = Abs(None=DotProduct(A=CameraVectorWS(), B=VertexNormalWS()))
  DepthFade(
   Opacity=Multiply(A=Multiply(A=@n1.A, B=@n3), B=SmoothStep(Min=@n4, Max=@n5, Value=@n6))
   FadeDistance=ScalarParameter() {parameterName=FadeDistance}
  )
  MaterialFunctionCall(Emissive=Multiply(A=VectorParameter() {parameterName=Color, defaultValue=<rgba 1 1 1 0>}.RGB, B=@n1.RGB)) {materialFunction=&MF_FX_EmissiveAdjustments}
 ]
}
